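# Material function for the RhombCamo player body that adds a switchable reactive color pulse to the incoming material attributes.
# Time since a game-time parameter, or a looping debug clock, samples a curve atlas to animate the pulse.
# A sphere mask on pre-skinned local position, combined with packed offset, reactivity and depth texture masks, shapes the effect.
# The masks darken base, emissive and subsurface color, blend in a boosted reactivity color, and lower opacity.
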MaterialFunction /Game/Characters/Player/Male/Medium/Bodies/M_MED_RhombCamo/Materials/MFC_RhombCamo_Reactivity
graph {nodes [
  Constant() {r=0.5}
  n1 = Multiply(A=Time(), B=ScalarParameter() {defaultValue=0.2, group=RhombCamo_Reactivity, parameterName=RC_Reactivity_DebugSpeed}) {constB=2}
  n2 = ScalarParameter() {defaultValue=5, group=RhombCamo_Reactivity, parameterName=RC_Reactivity_DebugSteps}
  Divide(A=Floor(None=Fmod(A=@n1, B=@n2)), B=@n2)
  FunctionOutput(None=StaticSwitchParameter(True=NamedRerouteUsage(), False=NamedRerouteUsage()) {group=RhombCamo, parameterName=UseRhombCamoReacivity})
  NamedRerouteDeclaration(None=FunctionInput() {inputType=FunctionInput_MaterialAttributes}) {name=MA_IN, nodeColor=<rgba 1 0 0.2344 1>}
  n3 = GetMaterialAttributes(None=NamedRerouteUsage()) {attributeGetTypes [
    #1
    #2
    #3
    #4
   ]}
  n4 = Power(Base=NamedRerouteUsage(), Exp=ScalarParameter() {defaultValue=0.5, group=RhombCamo_Reactivity, parameterName=RC_ReactivityMaskPower}) {constExponent=0.5}
  n5 = Saturate(None=@n4)
  n6 = LinearInterpolate(A=@n3.BaseColor, Alpha=@n5) {constB=0, desc=lerp with black}
  n7 = VectorParameter() {group=RhombCamo_Reactivity, parameterName=RC_ReactivityColor, defaultValue=<rgba 0.5906 0.0168 0.0027 1>}
  n8 = Multiply(A=@n7.RGB, B=ScalarParameter() {defaultValue=1.5, group=RhombCamo_Reactivity, parameterName=RC_EmissiveIntensity})
  n9 = NamedRerouteUsage()
  n10 = Multiply(A=@n4) {constB=3}
  n11 = SetMaterialAttributes!(
   MaterialAttributes=@n3.MaterialAttributes
   BaseColor=LinearInterpolate(A=StaticSwitchParameter(True=@n3.BaseColor, False=@n6) {group=RhombCamo_Reactivity, parameterName=UseRawBasecolorLerp}, B=@n8, Alpha=@n9)
   EmissiveColor=LinearInterpolate(A=LinearInterpolate(A=@n3.EmissiveColor, Alpha=@n5) {constB=0}, B=@n8, Alpha=@n9)
   SubsurfaceColor=LinearInterpolate(A=LinearInterpolate(A=@n3.SubsurfaceColor, Alpha=@n5) {constB=0}, B=@n8, Alpha=@n9)
   Opacity=LinearInterpolate(A=@n3.Opacity, Alpha=Saturate(None=@n10)) {constB=0}
  ) {attributeSetTypes [
    #1
    #2
    #3
    #4
   ]}
  NamedRerouteDeclaration(None=@n11) {name=MA_OUT, nodeColor=<rgba 0.1016 0 1 1>}
  n12 = PreSkinnedPosition()
  NamedRerouteDeclaration(None=FunctionInput(Preview=VertexInterpolator(VS=@n12).PS) {inputName=PreskinLocalPos, sortPriority=1}) {name=PreskinLocalPos, nodeColor=<rgba 1 0.75 0 1>}
  n13 = NamedRerouteUsage()
  n14 = Constant() {r=-0.5}
  n15 = Constant() {r=1.5}
  n16 = Multiply(
   A=SphereMask(
    A=MaterialFunctionCall(Input=@n13, Target Low=@n14, Target High=@n15) {materialFunction=&RemapValueRange}.Result
    B=NamedRerouteUsage()
    Radius=ScalarParameter() {defaultValue=0.5, group=RhombCamo_Reactivity, parameterName=RC_Reactivity_PulseSpread}
   ) {attenuationRadius=0.1, hardnessPercent=0}
   B=NamedRerouteUsage()
  )
  n17 = NamedRerouteDeclaration(None=Saturate(None=@n16)) {name=Reactivity_Raw, nodeColor=<rgba 0.9219 1 0 1>}
  n18 = NamedRerouteUsage()
  n19 = ScalarParameter() {defaultValue=0.5, group=RhombCamo_Reactivity, parameterName=RC_DepthIntensity}
  n20 = Multiply(
   A=Power(Base=@n16, Exp=ScalarParameter() {defaultValue=1.8, group=RhombCamo_Reactivity, parameterName=RC_Reactivity_PulseMaskPower})
   B=ConstantBiasScale(None=Multiply(A=@n18, B=@n19)) {bias=-0.5, scale=2}
  )
  NamedRerouteDeclaration(None=SmoothStep(Value=Add(A=@n17, B=@n20))) {name=Reactivity, nodeColor=<rgba 0 1 0.2266 1>}
  n21 = TextureObjectParameter() {
   group=RhombCamo_Reactivity
   parameterName=RC_Reactivity_Tex
   samplerType=SAMPLERTYPE_Masks
   texture=/Game/Characters/Player/Male/Medium/Bodies/M_MED_RhombCamo/Textures/T_RhomboCamo_Body_TA
  }
  n22 = TextureSample(Tex=@n21) {
   samplerSource=SSM_Wrap_WorldGroupSettings
   samplerType=SAMPLERTYPE_LinearColor
   texture=/Game/Characters/Player/Male/Medium/Bodies/M_MED_RhombCamo/Skins/Cyan/Textures/T_RhombCamo_Cyan_Body_TA
  }
  NamedRerouteDeclaration(None=@n22.R) {name=OffsetMask, nodeColor=<rgba 1 0 0.0399 1>}
  NamedRerouteDeclaration(None=@n22.G) {name=ReactivityMask, nodeColor=<rgba 0.0236 1 0 1>}
  n23 = Frac(None=@n1)
  n24 = ScalarParameter() {group=RhombCamo_Reactivity, parameterName=RC_GameTime_CCPM}
  n25 = LinearInterpolate(B=Subtract(A=Time(), B=@n24), Alpha=Sign(None=@n24))
  n26 = ScalarParameter() {defaultValue=2.55, group=RhombCamo_Reactivity, parameterName=RC_Reactivity_Duration}
  n27 = CurveAtlasRowParameter(CurveTime=StaticSwitchParameter(True=Saturate(None=@n23), False=Divide(A=@n25, B=@n26)) {group=RhombCamo_Reactivity, parameterName=UseDebugTime}) {
   atlas=/Game/Characters/Player/Male/Medium/Bodies/M_MED_RhombCamo/Textures/CA_RhombCamo
   curve=/Game/Characters/Player/Male/Medium/Bodies/M_MED_RhombCamo/Textures/C_RhombCamo
   group=RhombCamo_Reactivity
   parameterName=RC_Reactivity_Curve
  }
  n28 = Saturate(None=@n27)
  NamedRerouteDeclaration(None=LinearInterpolate(
    B=OneMinus(None=@n28)
    Alpha=ScalarParameter() {defaultValue=1, group=RhombCamo_Reactivity, parameterName=RC_ReactivityToggle, sortPriority=0}
   ) {constB=0}) {name=Time, nodeColor=<rgba 0 0.5312 1 1>}
  NamedRerouteDeclaration(None=@n22.B) {name=Depth, nodeColor=<rgba 1 0.8203 0 1>}
 ]}
