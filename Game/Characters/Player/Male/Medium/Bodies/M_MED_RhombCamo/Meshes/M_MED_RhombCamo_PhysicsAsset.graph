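# Physics asset for the medium male player RhombCamo body, set up for a dynamic jacket with capsule bodies on pelvis, lower spine, legs and feet.
# Front, side and rear jacket chains on each side hang three segments from small master-bone capsules.
# Chain joints allow no twist, and one swing limit widens toward the hem.
# Angularly free constraints link neighboring hem boxes within each side and thigh twist bones to thighs.
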PhysicsAsset /Game/Characters/Player/Male/Medium/Bodies/M_MED_RhombCamo/Meshes/M_MED_RhombCamo_PhysicsAsset
bodies {
 calf_l [{shapeType=Capsule, length=14, radius=8, shapeName=calf_r_capsule, center=<-7.7185 -2.0134 0.175>, rotation=<rot 90 0 0>}]
 calf_r [{shapeType=Capsule, length=14, radius=8, shapeName=calf_r_capsule, center=<6.6775 2.1026 -0.9979>, rotation=<rot 90 0 0>}]
 calf_twist_01_l [{shapeType=Capsule, length=21.5, radius=7.5, shapeName=calf_r_capsule, center=<5.9079 -1.1921 -0.9979>, rotation=<rot 90 7.125 17.1253>}]
 calf_twist_01_r [{shapeType=Capsule, length=21.5, radius=7.5, shapeName=calf_r_capsule, center=<-5.319 1.8804 -0.9979>, rotation=<rot 90 60.224 70.2243>}]
 dyn_jacket_front_01_l [{
   shapeType=Box
   extentX=13
   extentY=19
   extentZ=3
   shapeName=dyn_jacket_front_01_r_box
   center=<10.5278 0.1306 3.2>
   rotation=<rot -9.8463 -1.7539 10.1511>
  }]
 dyn_jacket_front_01_r [{
   shapeType=Box
   extentX=13
   extentY=19
   extentZ=3
   shapeName=dyn_jacket_front_01_r_box
   center=<10.5278 0.1306 0.7411>
   rotation=<rot -9.8463 -1.7539 10.1511>
  }]
 dyn_jacket_front_02_l [{shapeType=Box, extentX=23, extentY=17, extentZ=5, shapeName=dyn_jacket_front_02_l_box, center=<6.3781 0 0.0837>}]
 dyn_jacket_front_02_r [{shapeType=Box, extentX=18, extentY=17, extentZ=4, shapeName=dyn_jacket_front_02_r_box, center=<6.3781 0 0>}]
 dyn_jacket_front_03_l [{shapeType=Box, extentX=28, extentY=36, extentZ=5, shapeName=dyn_jacket_front_03_l_box, center=<9.2449 0 0>, rotation=<rot 0.0001 0 -10>}]
 dyn_jacket_front_03_r [{shapeType=Box, extentX=28, extentY=35, extentZ=5, shapeName=dyn_jacket_front_03_r_box, center=<9.2449 0 0>}]
 dyn_jacket_rear_01_l [{shapeType=Capsule, length=9.5, radius=7, shapeName=dyn_jacket_rear_01_l_capsule, center=<10.2351 0 6.092>, rotation=<rot -0.0002 0.0001 -90>}]
 dyn_jacket_rear_01_r [{
   shapeType=Capsule
   length=9.5
   radius=7
   shapeName=dyn_jacket_rear_01_l_capsule
   center=<10.2351 0 6.092>
   rotation=<rot -0.0002 -0.0001 -90>
  }]
 dyn_jacket_rear_02_l [{
   shapeType=Capsule
   length=15.5
   radius=7.5
   shapeName=dyn_jacket_rear_01_l_capsule
   center=<6.6404 -0.9708 4.8>
   rotation=<rot -0.0003 0.0002 80>
  }]
 dyn_jacket_rear_02_r [{
   shapeType=Capsule
   length=15.5
   radius=7.5
   shapeName=dyn_jacket_rear_01_l_capsule
   center=<7.4502 0.9708 4.8>
   rotation=<rot -0.0003 -0.0002 100>
  }]
 dyn_jacket_rear_03_l [{
   shapeType=Box
   extentX=28
   extentY=43
   extentZ=5
   shapeName=dyn_jacket_rear_03_l_box
   center=<9.2449 -3.2693 0>
   rotation=<rot -0.0001 0 -20>
  }]
 dyn_jacket_rear_03_r [{
   shapeType=Box
   extentX=28
   extentY=43
   extentZ=5
   shapeName=dyn_jacket_rear_03_r_box
   center=<9.2449 3.7318 0.2368>
   rotation=<rot -0.0002 -0.0001 20>
  }]
 dyn_jacket_side_01_l [{shapeType=Capsule, length=9.5, radius=7, shapeName=dyn_jacket_side_01_l_capsule, center=<10.2351 0 6.092>, rotation=<rot 0 0.0002 -90.0001>}]
 dyn_jacket_side_01_r [{shapeType=Capsule, length=9.5, radius=7, shapeName=dyn_jacket_side_01_l_capsule, center=<10.2351 0 6.092>, rotation=<rot 0 -0.0002 -89.9999>}]
 dyn_jacket_side_02_l [{
   shapeType=Capsule
   length=15.5
   radius=7.5
   shapeName=dyn_jacket_side_02_l_capsule
   center=<7.1464 -2.7954 4.8>
   rotation=<rot 0 0.0005 79.9999>
  }]
 dyn_jacket_side_02_r [{
   shapeType=Capsule
   length=15.5
   radius=7.5
   shapeName=dyn_jacket_side_02_l_capsule
   center=<6.075 2.7954 4.8>
   rotation=<rot 0.0001 -0.0008 100.0002>
  }]
 dyn_jacket_side_03_l [{shapeType=Box, extentX=28, extentY=22, extentZ=5, shapeName=dyn_jacket_side_03_l_box, center=<9.2449 0 0>}]
 dyn_jacket_side_03_r [{shapeType=Box, extentX=28, extentY=35, extentZ=5, shapeName=dyn_jacket_side_03_r_box, center=<9.2449 0 0>}]
 foot_l [{
   shapeType=Capsule
   length=18.8263
   radius=7.899
   shapeName=foot_r_capsule
   center=<-3.8984 6.0897 1.0346>
   rotation=<rot 0.0079 -0.4187 -93.2>
  }]
 foot_r [{
   shapeType=Capsule
   length=18.8263
   radius=7.899
   shapeName=foot_r_capsule
   center=<3.3729 -5.8893 0.3689>
   rotation=<rot 0.0079 -0.4187 -93.2>
  }]
 mstr_Jacket_Left [{shapeType=Capsule, length=4, radius=1, shapeName=mstr_Jacket_Left_capsule, center=<-0.4114 0 0>}]
 mstr_Jacket_Right [{shapeType=Capsule, length=4, radius=1, shapeName=mstr_Jacket_Right_capsule, center=<-0.4114 0 0>, rotation=<rot 0 0 90>}]
 mstr_jacket_front_01_l [{shapeType=Capsule, length=4, radius=1, shapeName=mstr_jacket_front_01_l_capsule, center=<-0.4114 0 0>, rotation=<rot 0 0 90>}]
 mstr_jacket_front_01_r [{shapeType=Capsule, length=4, radius=1, shapeName=mstr_jacket_front_01_r_capsule, center=<-0.4114 0 0>, rotation=<rot 0 0 90>}]
 mstr_jacket_rear_01_l [{shapeType=Capsule, length=0.505, radius=0.505, shapeName=mstr_jacket_rear_01_l_capsule, rotation=<rot 0 -90 0>}]
 mstr_jacket_rear_01_r [{shapeType=Capsule, length=0.505, radius=0.505, shapeName=mstr_jacket_rear_01_r_capsule, rotation=<rot 0 -90 0>}]
 mstr_jacket_side_01_l [{shapeType=Capsule, length=4, radius=1, shapeName=mstr_jacket_side_01_l_capsule, center=<-0.4114 0 0>, rotation=<rot 0 0 90>}]
 mstr_jacket_side_01_r [{shapeType=Capsule, length=4, radius=1, shapeName=mstr_jacket_side_01_r_capsule, center=<-0.4114 0 0>, rotation=<rot 0 0 90>}]
 pelvis [{shapeType=Capsule, length=7, radius=10.4, shapeName=spine_01_capsule, center=<-0.4269 -0.4086 -0.1905>}]
 spine_01 [{shapeType=Capsule, length=4.1, radius=10.4, shapeName=spine_01_capsule, center=<-0.4269 0.4787 0>}]
 spine_02 [{shapeType=Capsule, length=1.6, radius=10.4, shapeName=spine_01_capsule, center=<-0.4532 0.808 0>}]
 thigh_l [{
   shapeType=Capsule
   length=28
   radius=7.5
   shapeName=thigh_l_capsule
   center=<-17.8502 -1.8467 0.5056>
   rotation=<rot 89.8457 179.7214 -0.2985>
  }]
 thigh_r [{shapeType=Capsule, length=49, radius=8.5, shapeName=thigh_l_capsule, center=<17.8325 -0.7856 -0.7142>, rotation=<rot 89 179.9989 -0.0003>}]
 thigh_twist_01_l [{shapeType=Capsule, length=13.384, radius=8.0296, shapeName=thigh_twist_01_l_capsule, rotation=<rot 90 0 0>}]
 thigh_twist_01_r [{shapeType=Capsule, length=15.0416, radius=9.0413, shapeName=thigh_twist_01_r_capsule, rotation=<rot 90 0 0>}]
}
constraints [
 {bone1Name=dyn_jacket_front_01_l, bone2Name=mstr_jacket_front_01_l, swing1LimitDegrees=15, swing2LimitDegrees=25, twistLimitDegrees=0}
 {bone1Name=dyn_jacket_front_02_l, bone2Name=dyn_jacket_front_01_l, swing1LimitDegrees=15, swing2LimitDegrees=36, twistLimitDegrees=0}
 {bone1Name=dyn_jacket_front_03_l, bone2Name=dyn_jacket_front_02_l, swing1LimitDegrees=15, swing2LimitDegrees=68, twistLimitDegrees=0}
 {bone1Name=dyn_jacket_side_03_l, bone2Name=dyn_jacket_side_02_l, swing1LimitDegrees=15, swing2LimitDegrees=68, twistLimitDegrees=0}
 {bone1Name=dyn_jacket_side_02_l, bone2Name=dyn_jacket_side_01_l, swing1LimitDegrees=15, swing2LimitDegrees=36, twistLimitDegrees=0}
 {bone1Name=dyn_jacket_side_01_l, bone2Name=mstr_jacket_side_01_l, swing1LimitDegrees=15, swing2LimitDegrees=25, twistLimitDegrees=0}
 {bone1Name=dyn_jacket_front_03_r, bone2Name=dyn_jacket_front_02_r, swing1LimitDegrees=15, swing2LimitDegrees=68, twistLimitDegrees=0}
 {bone1Name=dyn_jacket_front_02_r, bone2Name=dyn_jacket_front_01_r, swing1LimitDegrees=15, swing2LimitDegrees=36, twistLimitDegrees=0}
 {bone1Name=dyn_jacket_front_01_r, bone2Name=mstr_jacket_front_01_r, swing1LimitDegrees=15, swing2LimitDegrees=25, twistLimitDegrees=0}
 {bone1Name=dyn_jacket_side_03_r, bone2Name=dyn_jacket_side_02_r, swing1LimitDegrees=15, swing2LimitDegrees=68, twistLimitDegrees=0}
 {bone1Name=dyn_jacket_side_02_r, bone2Name=dyn_jacket_side_01_r, swing1LimitDegrees=15, swing2LimitDegrees=36, twistLimitDegrees=0}
 {bone1Name=dyn_jacket_side_01_r, bone2Name=mstr_jacket_side_01_r, swing1LimitDegrees=15, swing2LimitDegrees=25, twistLimitDegrees=0}
 {bone1Name=dyn_jacket_rear_01_l, bone2Name=mstr_jacket_rear_01_l, swing1LimitDegrees=15, swing2LimitDegrees=25, twistLimitDegrees=0}
 {bone1Name=dyn_jacket_rear_02_l, bone2Name=dyn_jacket_rear_01_l, swing1LimitDegrees=15, swing2LimitDegrees=36, twistLimitDegrees=0}
 {bone1Name=dyn_jacket_rear_03_l, bone2Name=dyn_jacket_rear_02_l, swing1LimitDegrees=15, swing2LimitDegrees=68, twistLimitDegrees=0}
 {bone1Name=dyn_jacket_rear_01_r, bone2Name=mstr_jacket_rear_01_r, swing1LimitDegrees=15, swing2LimitDegrees=25, twistLimitDegrees=0}
 {bone1Name=dyn_jacket_rear_02_r, bone2Name=dyn_jacket_rear_01_r, swing1LimitDegrees=15, swing2LimitDegrees=36, twistLimitDegrees=0}
 {bone1Name=dyn_jacket_rear_03_r, bone2Name=dyn_jacket_rear_02_r, swing1LimitDegrees=15, swing2LimitDegrees=68, twistLimitDegrees=0}
 {
  bone1Name=dyn_jacket_rear_03_r
  bone2Name=dyn_jacket_side_03_r
  swing1LimitDegrees=45
  swing1Motion=Free
  swing2LimitDegrees=45
  swing2Motion=Free
  twistLimitDegrees=45
  twistMotion=Free
 }
 {
  bone1Name=thigh_twist_01_r
  bone2Name=thigh_r
  swing1LimitDegrees=45
  swing1Motion=Free
  swing2LimitDegrees=45
  swing2Motion=Free
  twistLimitDegrees=45
  twistMotion=Free
 }
 {
  bone1Name=thigh_twist_01_l
  bone2Name=thigh_l
  swing1LimitDegrees=45
  swing1Motion=Free
  swing2LimitDegrees=45
  swing2Motion=Free
  twistLimitDegrees=45
  twistMotion=Free
 }
 {
  bone1Name=dyn_jacket_side_03_r
  bone2Name=dyn_jacket_front_03_r
  swing1LimitDegrees=45
  swing1Motion=Free
  swing2LimitDegrees=45
  swing2Motion=Free
  twistLimitDegrees=45
  twistMotion=Free
 }
 {
  bone1Name=dyn_jacket_side_03_l
  bone2Name=dyn_jacket_front_03_l
  swing1LimitDegrees=45
  swing1Motion=Free
  swing2LimitDegrees=45
  swing2Motion=Free
  twistLimitDegrees=45
  twistMotion=Free
 }
 {
  bone1Name=dyn_jacket_rear_03_l
  bone2Name=dyn_jacket_side_03_l
  swing1LimitDegrees=45
  swing1Motion=Free
  swing2LimitDegrees=45
  swing2Motion=Free
  twistLimitDegrees=45
  twistMotion=Free
 }
]
