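# Slide-variant physics asset for the medium male RhombCamo player body, set up for physics on the jacket's dynamic bones.
# Three-segment front, side and rear capsule chains per side hang from small master-bone bodies via constraints, with side-to-front links at the middle segments.
# Capsules on pelvis, lower spine, thighs, calves and feet provide body collision shapes.
# A huge root sphere, offset downward to peak just above the origin, approximates a floor.
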PhysicsAsset /Game/Characters/Player/Male/Medium/Bodies/M_MED_RhombCamo/Meshes/M_MED_RhombCamo_PhysicsAsset_Slide
bodies {
 calf_l [{shapeType=Capsule, length=22.2, radius=8.3, shapeName=calf_r_capsule, center=<-15.3385 -1.3606 0.175>, rotation=<rot 90 180 4.8963>}]
 calf_r [{shapeType=Capsule, length=22.2, radius=9, shapeName=calf_r_capsule, center=<6.6775 2.1026 -0.9979>, rotation=<rot 90 180 -175.1037>}]
 calf_twist_01_l [{shapeType=Capsule, length=22.6, radius=6.4, shapeName=calf_r_capsule, center=<0.2341 -1.455 0.175>, rotation=<rot 90 180 4.8963>}]
 calf_twist_01_r [{shapeType=Capsule, length=24.4, radius=9.1, shapeName=calf_r_capsule, center=<-0.2341 1.455 -0.9979>, rotation=<rot 90 60.224 57.8001>}]
 dyn_jacket_front_01_l [{shapeType=Capsule, length=6.3, radius=11.7, shapeName=dyn_jacket_front_01_l_capsule, center=<8.994 0 11.5663>, rotation=<rot 0 0 -90>}]
 dyn_jacket_front_01_r [{shapeType=Capsule, length=9.1, radius=7.9, shapeName=dyn_jacket_front_01_l_capsule, center=<8 0 9>, rotation=<rot 0 0 -90>}]
 dyn_jacket_front_02_l [{shapeType=Capsule, length=6.5, radius=14.8, shapeName=dyn_jacket_front_02_l_capsule, center=<4.5923 0 12.9096>, rotation=<rot 0 0 90>}]
 dyn_jacket_front_02_r [{shapeType=Capsule, length=9, radius=8.2, shapeName=dyn_jacket_front_02_l_capsule, center=<7 0 8.2>, rotation=<rot 0 0 -90>}]
 dyn_jacket_front_03_l [{shapeType=Capsule, length=7.5, radius=10.3, shapeName=dyn_jacket_front_03_l_capsule, center=<4.5 0 7.8901>, rotation=<rot 0 0 90>}]
 dyn_jacket_front_03_r [{shapeType=Capsule, length=9, radius=7.4, shapeName=dyn_jacket_front_03_l_capsule, center=<4.5 0 8>, rotation=<rot 0 0 -90>}]
 dyn_jacket_rear_01_l [{
   shapeType=Capsule
   length=2.3175
   radius=25.0675
   shapeName=dyn_jacket_rear_01_l_capsule
   center=<4.2319 -1.1178 26.6724>
   rotation=<rot -0.028 -2.1489 -90.7418>
  }]
 dyn_jacket_rear_01_r [{
   shapeType=Capsule
   length=14.9175
   radius=12.7675
   shapeName=dyn_jacket_rear_01_l_capsule
   center=<4.4569 -0.4126 14.2544>
   rotation=<rot -0.3992 3.1027 94.0269>
  }]
 dyn_jacket_rear_02_l [{
   shapeType=Capsule
   length=15.7175
   radius=14.8675
   shapeName=dyn_jacket_rear_01_l_capsule
   center=<9.3897 0.1258 16.5818>
   rotation=<rot -0.028 -2.1489 -90.7418>
  }]
 dyn_jacket_rear_02_r [{
   shapeType=Capsule
   length=23.7175
   radius=10.6675
   shapeName=dyn_jacket_rear_01_l_capsule
   center=<6.634 -0.4224 11.5414>
   rotation=<rot -0.4314 3.1005 94.3783>
  }]
 dyn_jacket_rear_03_l [{
   shapeType=Capsule
   length=15.5373
   radius=12.8658
   shapeName=dyn_jacket_rear_03_l_capsule
   center=<6.0551 -0.0599 12.1851>
   rotation=<rot -0.028 -2.1489 -90.7418>
  }]
 dyn_jacket_rear_03_r [{
   shapeType=Capsule
   length=14.6373
   radius=13.0658
   shapeName=dyn_jacket_rear_03_l_capsule
   center=<6.6333 -0.4456 11.9039>
   rotation=<rot -0.2926 3.1092 92.868>
  }]
 dyn_jacket_side_01_l [{shapeType=Capsule, length=9.7, radius=8.3, shapeName=dyn_jacket_side_01_l_capsule, center=<6 0.9542 9.6855>, rotation=<rot 0 0 79>}]
 dyn_jacket_side_01_r [{shapeType=Capsule, length=9.8, radius=8.2, shapeName=dyn_jacket_side_01_l_capsule, center=<9.2333 0.0527 9.7478>, rotation=<rot 0 0 -78>}]
 dyn_jacket_side_02_l [{shapeType=Capsule, length=9.9, radius=8.1, shapeName=dyn_jacket_side_02_l_capsule, center=<5.1 0.1384 9.6269>, rotation=<rot 0 0 79>}]
 dyn_jacket_side_02_r [{shapeType=Capsule, length=9, radius=8.6, shapeName=dyn_jacket_side_02_l_capsule, center=<5.1 0.1661 10.3812>, rotation=<rot 0 0 -78>}]
 dyn_jacket_side_03_l [{shapeType=Capsule, length=7.2, radius=10.8, shapeName=dyn_jacket_side_03_l_capsule, center=<5.75 0.2717 11.172>, rotation=<rot 0 0 79>}]
 dyn_jacket_side_03_r [{shapeType=Capsule, length=11, radius=5.9, shapeName=dyn_jacket_side_04_l_capsule, center=<3.673 -0.1664 7.4892>, rotation=<rot 0 0 -78>}]
 foot_l [{shapeType=Capsule, length=15.6263, radius=9.399, shapeName=foot_r_capsule, center=<-4.0894 6.251 0>, rotation=<rot 0.0079 -0.4187 86.8>}]
 foot_r [{
   shapeType=Capsule
   length=16.3263
   radius=10.399
   shapeName=foot_r_capsule
   center=<8.0755 -5.9236 0.3695>
   rotation=<rot 0.0079 -0.4187 -93.2>
  }]
 mstr_Jacket_Left [{shapeType=Capsule, length=4, radius=1, shapeName=mstr_Jacket_Left_capsule, center=<-0.4114 0 0>}]
 mstr_Jacket_Right [{shapeType=Capsule, length=4, radius=1, shapeName=mstr_Jacket_Right_capsule, center=<-0.4114 0 0>, rotation=<rot 0 0 90>}]
 mstr_jacket_front_01_l [{shapeType=Capsule, length=4, radius=1, shapeName=mstr_jacket_front_01_l_capsule, center=<-0.4114 0 0>, rotation=<rot 0 0 90>}]
 mstr_jacket_front_01_r [{shapeType=Capsule, length=4, radius=1, shapeName=mstr_jacket_front_01_r_capsule, center=<-0.4114 0 0>, rotation=<rot 0 0 90>}]
 mstr_jacket_rear_01_l [{shapeType=Capsule, length=0.505, radius=0.505, shapeName=mstr_jacket_rear_01_l_capsule, rotation=<rot 0 -90 0>}]
 mstr_jacket_rear_01_r [{shapeType=Capsule, length=0.505, radius=0.505, shapeName=mstr_jacket_rear_01_r_capsule, rotation=<rot 0 -90 0>}]
 mstr_jacket_side_01_l [{shapeType=Capsule, length=4, radius=1, shapeName=mstr_jacket_side_01_l_capsule, center=<-0.4114 0 0>, rotation=<rot 0 0 90>}]
 mstr_jacket_side_01_r [{shapeType=Capsule, length=4, radius=1, shapeName=mstr_jacket_side_01_r_capsule, center=<-0.4114 0 0>, rotation=<rot 0 0 90>}]
 pelvis [{shapeType=Capsule, length=9, radius=10.4, shapeName=spine_01_capsule, center=<-0.4269 -0.4086 -0.1905>}]
 root [{shapeType=Sphere, radius=1000, shapeName=root_sphere, center=<0 0 -990>}]
 spine_01 [{shapeType=Capsule, length=7.6, radius=10.4, shapeName=spine_01_capsule, center=<-0.4269 0.4787 0>}]
 spine_02 [{shapeType=Capsule, length=5.6, radius=10.4, shapeName=spine_01_capsule, center=<-0.4532 0.808 0>}]
 thigh_l [{
   shapeType=Capsule
   length=24.4
   radius=10.1
   shapeName=thigh_l_capsule
   center=<-17.8502 -1.8467 0.5056>
   rotation=<rot 89.8457 179.7214 -0.2985>
  }]
 thigh_r [{shapeType=Capsule, length=37.5, radius=11, shapeName=thigh_l_capsule, center=<15.3825 0.0025 -0.8942>, rotation=<rot 85.7977 179.9995 0>}]
}
constraints [
 {
  bone1Name=dyn_jacket_front_01_l
  bone2Name=mstr_jacket_front_01_l
  swing1LimitDegrees=40
  swing1Motion=Free
  swing2LimitDegrees=40
  twistLimitDegrees=40
  twistMotion=Free
 }
 {bone1Name=dyn_jacket_front_02_l, bone2Name=dyn_jacket_front_01_l, swing1LimitDegrees=10, swing2LimitDegrees=40, twistLimitDegrees=5}
 {
  bone1Name=dyn_jacket_front_03_l
  bone2Name=dyn_jacket_front_02_l
  swing1LimitDegrees=10
  swing1Motion=Free
  swing2LimitDegrees=30
  swing2Motion=Free
  twistLimitDegrees=5
  twistMotion=Free
 }
 {bone1Name=dyn_jacket_side_03_l, bone2Name=dyn_jacket_side_02_l, swing1LimitDegrees=7, swing2LimitDegrees=30, twistLimitDegrees=10}
 {bone1Name=dyn_jacket_side_02_l, bone2Name=dyn_jacket_side_01_l, swing1LimitDegrees=7, swing2LimitDegrees=25, twistLimitDegrees=10}
 {bone1Name=dyn_jacket_side_01_l, bone2Name=mstr_jacket_side_01_l, swing1LimitDegrees=50, swing2LimitDegrees=50, twistLimitDegrees=10}
 {bone1Name=dyn_jacket_front_03_r, bone2Name=dyn_jacket_front_02_r, swing1LimitDegrees=10, swing2LimitDegrees=50, twistLimitDegrees=5}
 {bone1Name=dyn_jacket_front_02_r, bone2Name=dyn_jacket_front_01_r, swing1LimitDegrees=10, swing2LimitDegrees=50, twistLimitDegrees=5}
 {
  bone1Name=dyn_jacket_front_01_r
  bone2Name=mstr_jacket_front_01_r
  swing1LimitDegrees=40
  swing1Motion=Free
  swing2LimitDegrees=50
  twistLimitDegrees=40
  twistMotion=Free
 }
 {bone1Name=dyn_jacket_side_03_r, bone2Name=dyn_jacket_side_02_r, swing1LimitDegrees=7, swing2LimitDegrees=30, twistLimitDegrees=10}
 {bone1Name=dyn_jacket_side_02_r, bone2Name=dyn_jacket_side_01_r, swing1LimitDegrees=7, swing2LimitDegrees=25, twistLimitDegrees=10}
 {bone1Name=dyn_jacket_side_01_r, bone2Name=mstr_jacket_side_01_r, swing1LimitDegrees=7, swing2LimitDegrees=25, twistLimitDegrees=10}
 {
  bone1Name=dyn_jacket_side_02_r
  bone2Name=dyn_jacket_front_02_r
  swing1LimitDegrees=45
  swing1Motion=Free
  swing2LimitDegrees=45
  swing2Motion=Free
  twistLimitDegrees=45
  twistMotion=Free
 }
 {bone1Name=dyn_jacket_side_02_l, bone2Name=dyn_jacket_front_02_l, swing1LimitDegrees=45, swing2LimitDegrees=45, twistLimitDegrees=45}
 {bone1Name=dyn_jacket_rear_01_l, bone2Name=mstr_jacket_rear_01_l, swing1LimitDegrees=50, swing2LimitDegrees=50, twistLimitDegrees=10}
 {bone1Name=dyn_jacket_rear_02_l, bone2Name=dyn_jacket_rear_01_l, swing1LimitDegrees=7, swing2LimitDegrees=25, twistLimitDegrees=10}
 {bone1Name=dyn_jacket_rear_03_l, bone2Name=dyn_jacket_rear_02_l, swing1LimitDegrees=7, swing2LimitDegrees=30, twistLimitDegrees=10}
 {bone1Name=dyn_jacket_rear_01_r, bone2Name=mstr_jacket_rear_01_r, swing1LimitDegrees=50, swing2LimitDegrees=50, twistLimitDegrees=10}
 {bone1Name=dyn_jacket_rear_02_r, bone2Name=dyn_jacket_rear_01_r, swing1LimitDegrees=7, swing2LimitDegrees=25, twistLimitDegrees=10}
 {bone1Name=dyn_jacket_rear_03_r, bone2Name=dyn_jacket_rear_02_r, swing1LimitDegrees=7, swing2LimitDegrees=30, twistLimitDegrees=10}
]
