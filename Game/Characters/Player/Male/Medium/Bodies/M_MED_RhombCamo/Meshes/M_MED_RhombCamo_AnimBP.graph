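# Post-copy AnimBP for the M_MED_RhombCamo player body, driving jacket secondary motion.
# The copied parent pose runs through Control Rig, then additive bone correctives and RigidBody physics with front-end or gameplay assets, plus further correctives.
# Sliding switches to a separate corrective and physics branch.
# EventGraph derives alphas from speed, thigh twist and movement state, sets skydive-aware gravity and feeds leg and wind data to WPO cloth materials.
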AnimBlueprint /Game/Characters/Player/Male/Medium/Bodies/M_MED_RhombCamo/Meshes/M_MED_RhombCamo_AnimBP
target_skeleton=&/M_MED_RhombCamo_Skeleton
graphs [
 AnimGraph {nodes [
   n1 = Anim Dynamics(Alpha=1, ExternalForce="0,0,0") {
    angularDampingOverride=1
    angularSpringConstant=250
    bAngularSpring=true
    bOverrideAngularDamping=true
    bUseGravityOverride=true
    gravityScale=0.35
    simulationSpace=RootRelative
    alphaBoolBlend {blendOption=Linear}
    componentAppliedLinearAccClamp=<3000 3000 3000>
    componentLinearAccScale=<0.5 0.4 0.4>
    componentLinearVelScale=<0.5 0.4 0.4>
    gravityOverride {z=-500}
    physicsBodyDefinitions [{
      boundBone {boneName=None}
      boxExtents=<30 30 30>
      constraintSetup {angularLimitsMax=<10 10 1>, angularLimitsMin=<-10 -10 -1>, angularTarget=<1 0 0>}
      localJointOffset=<-10 0 0>
     }]
    relativeSpaceBone {boneName=spine_02}
    boundBone=~
   }
   Anim Dynamics(ComponentPose=@n1, Alpha=1, ExternalForce="0,0,0") {
    angularDampingOverride=1
    angularSpringConstant=250
    bAngularSpring=true
    bOverrideAngularDamping=true
    bUseGravityOverride=true
    gravityScale=0.35
    simulationSpace=RootRelative
    alphaBoolBlend {blendOption=Linear}
    componentAppliedLinearAccClamp=<3000 3000 3000>
    componentLinearAccScale=<0.5 0.4 0.4>
    componentLinearVelScale=<0.5 0.4 0.4>
    gravityOverride {z=-500}
    physicsBodyDefinitions [{
      boundBone {boneName=None}
      boxExtents=<30 30 30>
      constraintSetup {angularLimitsMax=<10 10 1>, angularLimitsMin=<-10 -10 -1>, angularTarget=<1 0 0>}
      localJointOffset=<-10 0 0>
     }]
    relativeSpaceBone {boneName=spine_02}
    boundBone=~
   }
   n2 = "Transform (Modify) Bone - Bone: mstr_jacket_rear_01_r"(Alpha=Get movingRightSideAlpha() -> (movingRightSideAlpha:double), Translation="0,0,0", Rotation="10,0,0", Scale="1,1,1") {
    rotationMode=BMM_Additive
    rotationSpace=BCS_BoneSpace
    translationMode=BMM_Additive
    translationSpace=BCS_BoneSpace
    alphaBoolBlend {blendOption=Linear}
    alphaScaleBiasClamp {bInterpResult=true}
    boneToModify {boneName=mstr_jacket_rear_01_r}
    rotation {pitch=10}
   }
   n3 = "Transform (Modify) Bone - Bone: mstr_jacket_rear_01_l"(ComponentPose=@n2, Alpha=Get movingLeftSideAlpha() -> (movingLeftSideAlpha:double), Translation="0,0,0", Rotation="10,0,0", Scale="1,1,1") {
    rotationMode=BMM_Additive
    rotationSpace=BCS_BoneSpace
    translationMode=BMM_Additive
    translationSpace=BCS_BoneSpace
    alphaBoolBlend {blendOption=Linear}
    alphaScaleBiasClamp {bInterpResult=true}
    boneToModify {boneName=mstr_jacket_rear_01_l}
    rotation {pitch=10}
   }
   n4 = "Transform (Modify) Bone - Bone: dyn_jacket_rear_03_r"(
    ComponentPose=@n3
    Alpha=Get IsCrouchingAlpha() -> (IsCrouchingAlpha:double)
    Translation="-0.152,0.170,0.789"
    Rotation="0.053,-4.927,-0.016"
    Scale="1,1,1"
   ) {
    rotationMode=BMM_Additive
    rotationSpace=BCS_BoneSpace
    translationMode=BMM_Additive
    translationSpace=BCS_BoneSpace
    alphaBoolBlend {blendOption=Linear}
    alphaScaleBiasClamp {bInterpResult=true}
    boneToModify {boneName=dyn_jacket_rear_03_r}
    rotation=<rot 0.053 -4.927 -0.016>
    translation=<-0.152 0.17 0.789>
   }
   n5 = "Transform (Modify) Bone - Bone: mstr_jacket_rear_01_r"(ComponentPose=@n4, Alpha=Get pawnSpeedRange() -> (pawnSpeedRange:double), Translation="0,0,0", Rotation="10,0,0", Scale="1,1,1") {
    rotationMode=BMM_Additive
    rotationSpace=BCS_BoneSpace
    translationMode=BMM_Additive
    translationSpace=BCS_BoneSpace
    alphaBoolBlend {blendOption=Linear}
    alphaScaleBiasClamp {bInterpResult=true}
    boneToModify {boneName=mstr_jacket_rear_01_r}
    rotation {pitch=10}
   }
   n6 = "Transform (Modify) Bone - Bone: dyn_jacket_rear_01_r"(ComponentPose=@n5, Alpha=Get IsCrouchingAlpha(), Translation="-3,0,0", Rotation="3.301,12.599,-6.643", Scale="1,1,1") {
    rotationMode=BMM_Additive
    rotationSpace=BCS_BoneSpace
    translationMode=BMM_Additive
    translationSpace=BCS_BoneSpace
    alphaBoolBlend {blendOption=Linear}
    alphaScaleBiasClamp {bInterpResult=true}
    boneToModify {boneName=dyn_jacket_rear_01_r}
    rotation=<rot 3.301 12.599 -6.643>
    translation {x=-3}
   }
   n7 = "Transform (Modify) Bone - Bone: dyn_jacket_front_02_l"(ComponentPose=@n6, Alpha=Get IsCrouchingAlpha(), Translation="0,0,0", Rotation="-27.179,10.672,-0.907", Scale="1,1,1") {
    rotationMode=BMM_Additive
    rotationSpace=BCS_BoneSpace
    translationMode=BMM_Additive
    translationSpace=BCS_BoneSpace
    alphaBoolBlend {blendOption=Linear}
    alphaScaleBiasClamp {bInterpResult=true}
    boneToModify {boneName=dyn_jacket_front_02_l}
    rotation=<rot -27.179 10.672 -0.907>
   }
   n8 = "Transform (Modify) Bone - Bone: dyn_jacket_front_03_l"(ComponentPose=@n7, Alpha=Get IsCrouchingAlpha(), Translation="0,0,0", Rotation="24.013,-13.617,-10.981", Scale="1,1,1") {
    rotationMode=BMM_Additive
    rotationSpace=BCS_BoneSpace
    translationMode=BMM_Additive
    translationSpace=BCS_BoneSpace
    alphaBoolBlend {blendOption=Linear}
    alphaScaleBiasClamp {bInterpResult=true}
    boneToModify {boneName=dyn_jacket_front_03_l}
    rotation=<rot 24.013 -13.617 -10.981>
   }
   n9 = "Transform (Modify) Bone - Bone: mstr_Jacket_Left"(ComponentPose=@n8, Alpha=Get IsCrouchingAlpha(), Translation="0,-1,0", Rotation="0,0,0", Scale="1,1,1") {
    rotationMode=BMM_Additive
    rotationSpace=BCS_BoneSpace
    translationMode=BMM_Additive
    translationSpace=BCS_BoneSpace
    alphaBoolBlend {blendOption=Linear}
    alphaScaleBiasClamp {bInterpResult=true}
    boneToModify {boneName=mstr_Jacket_Left}
    translation {y=-1}
   }
   n10 = "Transform (Modify) Bone - Bone: dyn_jacket_front_01_r"(ComponentPose=@n9, Alpha=Get pawnSpeedRange(), Translation="0,0,1", Rotation="0,0,0", Scale="1,1,1") {
    rotationMode=BMM_Additive
    rotationSpace=BCS_BoneSpace
    translationMode=BMM_Additive
    translationSpace=BCS_BoneSpace
    alphaBoolBlend {blendOption=Linear}
    alphaScaleBiasClamp {bInterpResult=true}
    boneToModify {boneName=dyn_jacket_front_01_r}
    translation {z=1}
   }
   n11 = "Transform (Modify) Bone - Bone: dyn_jacket_front_01_l"(ComponentPose=@n10, Alpha=Get pawnSpeedRange(), Translation="0,0,2", Rotation="0,0,0", Scale="1,1,1") {
    rotationMode=BMM_Additive
    rotationSpace=BCS_BoneSpace
    translationMode=BMM_Additive
    translationSpace=BCS_BoneSpace
    alphaBoolBlend {blendOption=Linear}
    alphaScaleBiasClamp {bInterpResult=true}
    boneToModify {boneName=dyn_jacket_front_01_l}
    translation {z=2}
   }
   "Transform (Modify) Bone - Bone: mstr_jacket_side_01_r"(ComponentPose=@n11, Alpha=Get JumpingAlpha() -> (JumpingAlpha:double), Translation="0,0,0", Rotation="25,0,0", Scale="1,1,1") {
    rotationMode=BMM_Additive
    rotationSpace=BCS_BoneSpace
    translationMode=BMM_Additive
    translationSpace=BCS_BoneSpace
    alphaBoolBlend {blendOption=Linear}
    alphaScaleBiasClamp {bInterpResult=true}
    boneToModify {boneName=mstr_jacket_side_01_r}
    rotation {pitch=25}
   }
   Output Pose(Result=Use cached pose 'RBANPose'()) {layerGroup=None}
   n12 = "Transform (Modify) Bone - Bone: mstr_jacket_front_01_r"(
    ComponentPose=Local To Component(LocalPose=Use cached pose 'PrePose'())
    Alpha=Get IsCrouchingAlpha()
    Translation="0,2,0"
    Rotation="-10,-10,-10"
    Scale="1,1,1"
   ) {
    rotationMode=BMM_Additive
    rotationSpace=BCS_ParentBoneSpace
    translationMode=BMM_Additive
    translationSpace=BCS_BoneSpace
    alphaBoolBlend {blendOption=Linear}
    boneToModify {boneName=mstr_jacket_front_01_r}
    rotation=<rot -10 -10 -10>
    translation {y=2}
   }
   n13 = Get pawnSpeedRange()
   n14 = "Transform (Modify) Bone - Bone: mstr_Jacket"(ComponentPose=@n12, Alpha=@n13, Translation="0,0,0", Rotation="0,20,0", Scale="1,1,1") {
    rotationMode=BMM_Additive
    rotationSpace=BCS_BoneSpace
    translationMode=BMM_Additive
    translationSpace=BCS_BoneSpace
    alphaBoolBlend {blendOption=Linear}
    boneToModify {boneName=mstr_Jacket}
    rotation {yaw=20}
   }
   n15 = "Transform (Modify) Bone - Bone: mstr_jacket_front_01_l"(ComponentPose=@n14, Alpha=@n13, Translation="0,0,0", Rotation="0,0,0", Scale="1,1,1") {
    rotationMode=BMM_Additive
    rotationSpace=BCS_BoneSpace
    translationMode=BMM_Additive
    translationSpace=BCS_BoneSpace
    alphaBoolBlend {blendOption=Linear}
    boneToModify {boneName=mstr_jacket_front_01_l}
   }
   n16 = "Transform (Modify) Bone - Bone: mstr_Jacket"(ComponentPose=@n15, Alpha=@n13, Translation="0,0,0", Rotation="0,5,0", Scale="1,1,1") {
    rotationMode=BMM_Additive
    rotationSpace=BCS_BoneSpace
    translationMode=BMM_Additive
    translationSpace=BCS_BoneSpace
    alphaBoolBlend {blendOption=Linear}
    boneToModify {boneName=mstr_Jacket}
    rotation {yaw=5}
   }
   n17 = Get thigh_rot_up_l() -> (thigh_rot_up_l:double)
   n18 = "Transform (Modify) Bone - Bone: mstr_jacket_side_01_l"(ComponentPose=@n16, Alpha=@n17, Translation="2.770,0.198,-1.256", Rotation="-2.268,-5.582,2.807", Scale="1,1,1") {
    rotationMode=BMM_Additive
    rotationSpace=BCS_BoneSpace
    translationMode=BMM_Additive
    translationSpace=BCS_BoneSpace
    alphaBoolBlend {blendOption=Linear}
    boneToModify {boneName=mstr_jacket_side_01_l}
    rotation=<rot -2.268 -5.582 2.807>
    translation=<2.77 0.198 -1.256>
   }
   n19 = "Transform (Modify) Bone - Bone: mstr_Jacket_Left"(ComponentPose=@n18, Alpha=@n17, Translation="-0.393,-0.921,-5", Rotation="0,0,31.355", Scale="1,1,1") {
    rotationMode=BMM_Additive
    rotationSpace=BCS_BoneSpace
    translationMode=BMM_Additive
    translationSpace=BCS_BoneSpace
    alphaBoolBlend {blendOption=Linear}
    boneToModify {boneName=mstr_Jacket_Left}
    rotation {roll=31.355}
    translation=<-0.393 -0.921 -5>
   }
   n20 = Get thigh_rot_up_r() -> (thigh_rot_up_r:double)
   n21 = "Transform (Modify) Bone - Bone: mstr_jacket_side_01_r"(ComponentPose=@n19, Alpha=@n20, Translation="0,0,0", Rotation="1.002,5.991,6.827", Scale="1,1,1") {
    rotationMode=BMM_Additive
    rotationSpace=BCS_BoneSpace
    translationMode=BMM_Additive
    translationSpace=BCS_BoneSpace
    alphaBoolBlend {blendOption=Linear}
    boneToModify {boneName=mstr_jacket_side_01_r}
    rotation=<rot 1.002 5.991 6.827>
   }
   n22 = "Transform (Modify) Bone - Bone: mstr_Jacket_Right"(ComponentPose=@n21, Alpha=@n20, Translation="0,0,6", Rotation="-9.125,1.622,-4.575", Scale="1,1,1") {
    rotationMode=BMM_Additive
    rotationSpace=BCS_BoneSpace
    translationMode=BMM_Additive
    translationSpace=BCS_BoneSpace
    alphaBoolBlend {blendOption=Linear}
    boneToModify {boneName=mstr_Jacket_Right}
    rotation=<rot -9.125 1.6219 -4.575>
    translation {z=6}
   }
   n23 = Get IsCrouchingAlpha()
   n24 = "Transform (Modify) Bone - Bone: mstr_jacket_front_01_l"(ComponentPose=@n22, Alpha=@n23, Translation="-4.6,-2,0", Rotation="0,0,0", Scale="1,1,1") {
    rotationMode=BMM_Additive
    rotationSpace=BCS_BoneSpace
    translationMode=BMM_Additive
    translationSpace=BCS_BoneSpace
    alphaBoolBlend {blendOption=Linear}
    alphaScaleBiasClamp {bInterpResult=true}
    boneToModify {boneName=mstr_jacket_front_01_l}
    translation=<-4.6 -2>
   }
   n25 = "Transform (Modify) Bone - Bone: mstr_Jacket"(ComponentPose=@n24, Alpha=@n23, Translation="0,0,0", Rotation="0,10,0", Scale="1,1,1") {
    rotationMode=BMM_Additive
    rotationSpace=BCS_BoneSpace
    translationMode=BMM_Additive
    translationSpace=BCS_BoneSpace
    alphaBoolBlend {blendOption=Linear}
    alphaScaleBiasClamp {bInterpResult=true}
    boneToModify {boneName=mstr_Jacket}
    rotation {yaw=10}
   }
   n26 = "Transform (Modify) Bone - Bone: mstr_jacket_rear_01_r"(ComponentPose=@n25, Alpha=@n23, Translation="0,0,0", Rotation="10.001,0,0", Scale="1,1,1") {
    rotationMode=BMM_Additive
    rotationSpace=BCS_BoneSpace
    translationMode=BMM_Additive
    translationSpace=BCS_BoneSpace
    alphaBoolBlend {blendOption=Linear}
    alphaScaleBiasClamp {bInterpResult=true}
    boneToModify {boneName=mstr_jacket_rear_01_r}
    rotation {pitch=10.001}
   }
   n27 = "Transform (Modify) Bone - Bone: mstr_Jacket"(ComponentPose=@n26, Alpha=@n23, Translation="0,1,0", Rotation="0,0,0", Scale="1,1,1") {
    rotationMode=BMM_Additive
    rotationSpace=BCS_BoneSpace
    translationMode=BMM_Additive
    translationSpace=BCS_BoneSpace
    alphaBoolBlend {blendOption=Linear}
    alphaScaleBiasClamp {bInterpResult=true}
    boneToModify {boneName=mstr_Jacket}
    translation {y=1}
   }
   Save cached pose 'PreRbdAdjustments'(Pose=Component To Local(ComponentPose=@n27)) {cacheName=PreRbdAdjustments}
   n28 = "Transform (Modify) Bone - Bone: mstr_Jacket"(
    ComponentPose=Local To Component(LocalPose=Use cached pose 'PrePose'())
    Alpha=1
    Translation="0.224,1.648,-0.000"
    Rotation="0,50,0"
    Scale="1,1.5,1"
   ) {
    rotationMode=BMM_Additive
    rotationSpace=BCS_BoneSpace
    scaleMode=BMM_Replace
    translationMode=BMM_Additive
    translationSpace=BCS_BoneSpace
    alphaBoolBlend {blendInTime=0.5, blendOption=Linear}
    boneToModify {boneName=mstr_Jacket}
    rotation {yaw=50}
    scale {y=1.5}
    translation=<0.224 1.648>
   }
   n29 = "Transform (Modify) Bone - Bone: mstr_Jacket"(ComponentPose=@n28, Alpha=1, Translation="0,0,0", Rotation="0,0,0", Scale="1,1,1") {
    rotationMode=BMM_Additive
    rotationSpace=BCS_BoneSpace
    translationMode=BMM_Additive
    translationSpace=BCS_BoneSpace
    alphaBoolBlend {blendInTime=0.5, blendOption=Linear}
    boneToModify {boneName=mstr_Jacket}
   }
   n30 = RigidBody!(ComponentPose=@n29, Alpha, OverrideWorldGravity:Vector="0,0,-980", ExternalForce) -> (Pose) {
    bClampLinearTranslationLimitToRefPose=true
    bOverrideWorldGravity=true
    overridePhysicsAsset=&/M_MED_RhombCamo_PhysicsAsset_Slide
    simulationSpace=BaseBoneSpace
    alphaBoolBlend {blendOption=Linear}
    alphaScaleBiasClamp {interpSpeedDecreasing=15, interpSpeedIncreasing=20}
    baseBoneRef {boneName=pelvis}
    componentAppliedLinearAccClamp=<3000 1000 1000>
    componentLinearAccScale {x=0.1}
    componentLinearVelScale {x=0.2}
    overrideWorldGravity {z=-980}
   }
   n31 = RigidBody(ComponentPose=Local To Component(LocalPose=Use cached pose 'PreRbdAdjustments'()), OverrideWorldGravity=Get Gravity Override()) {
    bClampLinearTranslationLimitToRefPose=true
    bOverrideWorldGravity=true
    overridePhysicsAsset=&/M_MED_RhombCamo_FE_PhysicsAsset
    simulationSpace=BaseBoneSpace
    alphaBoolBlend {blendOption=Linear}
    alphaScaleBiasClamp {interpSpeedDecreasing=15, interpSpeedIncreasing=20}
    baseBoneRef {boneName=pelvis}
    componentAppliedLinearAccClamp=<3000 3000 3000>
    componentLinearAccScale=<0.1 0.05 0.05>
    componentLinearVelScale=<0.1 0.05 0.05>
    overrideWorldGravity {z=-980}
    simSpaceSettings {
     maxAngularAcceleration=500
     maxAngularVelocity=3
     maxLinearAcceleration=700
     velocityScaleZ=0
     worldAlpha=1
     externalAngularVelocity {x=0.8}
     externalLinearVelocity {x=0.3}
    }
   }
   n32 = Component To Local(ComponentPose=@n31)
   n33 = RigidBody(ComponentPose=Local To Component(LocalPose=Use cached pose 'PreRbdAdjustments'()), OverrideWorldGravity=Get Gravity Override()) {
    bClampLinearTranslationLimitToRefPose=true
    bOverrideWorldGravity=true
    overridePhysicsAsset=&/M_MED_RhombCamo_PhysicsAsset
    simulationSpace=BaseBoneSpace
    alphaBoolBlend {blendOption=Linear}
    alphaScaleBiasClamp {interpSpeedDecreasing=15, interpSpeedIncreasing=20}
    baseBoneRef {boneName=pelvis}
    componentAppliedLinearAccClamp=<3000 3000 3000>
    componentLinearAccScale=<0.1 0.05 0.05>
    componentLinearVelScale=<0.1 0.05 0.05>
    overrideWorldGravity {z=-980}
    simSpaceSettings {
     maxAngularAcceleration=500
     maxAngularVelocity=3
     maxLinearAcceleration=700
     velocityScaleZ=0
     worldAlpha=1
     externalAngularVelocity {x=0.8}
     externalLinearVelocity {x=0.3}
    }
   }
   n34 = Component To Local(ComponentPose=@n33)
   n35 = Get bIsInFrontEnd()
   n36 = "Transform (Modify) Bone - Bone: mstr_jacket_side_01_r"(
    ComponentPose=Local To Component(LocalPose=Blend Poses by bool(BlendPose_0=@n32, BlendPose_1=@n34, BlendTime_0=0.01, BlendTime_1=0.3, bActiveValue=@n35) {blendType=Linear})
    Alpha=Get movingRightSideAlpha()
    Translation="0,0,3"
    Rotation="5.865,0.265,-0.206"
    Scale="1,1,1"
   ) {
    rotationMode=BMM_Additive
    rotationSpace=BCS_BoneSpace
    translationMode=BMM_Additive
    translationSpace=BCS_BoneSpace
    alphaBoolBlend {blendOption=Linear}
    alphaScaleBiasClamp {bInterpResult=true}
    boneToModify {boneName=mstr_jacket_side_01_r}
    rotation=<rot 5.865 0.265 -0.206>
    translation {z=3}
   }
   n37 = "Transform (Modify) Bone - Bone: dyn_jacket_front_01_r"(ComponentPose=@n36, Alpha=Get pawnSpeedRange(), Translation="0,5,0", Rotation="0,0,0", Scale="1,1,1") {
    rotationMode=BMM_Additive
    rotationSpace=BCS_BoneSpace
    translationMode=BMM_Additive
    translationSpace=BCS_BoneSpace
    alphaBoolBlend {blendOption=Linear}
    boneToModify {boneName=dyn_jacket_front_01_r}
    translation {y=5}
   }
   n38 = "Transform (Modify) Bone - Bone: mstr_jacket_front_01_r"(ComponentPose=@n37, Alpha=Get movingRightSideAlpha(), Translation="0,0,-0.2709", Rotation="5.839,3.915,0.473", Scale="1,1,1") {
    rotationMode=BMM_Additive
    rotationSpace=BCS_BoneSpace
    translationMode=BMM_Additive
    translationSpace=BCS_BoneSpace
    alphaBoolBlend {blendOption=Linear}
    alphaScaleBiasClamp {bInterpResult=true}
    boneToModify {boneName=mstr_jacket_front_01_r}
    rotation=<rot 5.8391 3.9149 0.473>
    translation {z=-0.2709}
   }
   n39 = "Transform (Modify) Bone - Bone: mstr_jacket_front_01_r"(ComponentPose=@n38, Alpha=Get BwdMovement() -> (BwdMovement:double), Translation="0,0,0", Rotation="5,3.855,10.234", Scale="1,1,1") {
    rotationMode=BMM_Additive
    rotationSpace=BCS_BoneSpace
    translationMode=BMM_Additive
    translationSpace=BCS_BoneSpace
    alphaBoolBlend {blendOption=Linear}
    alphaScaleBiasClamp {bInterpResult=true}
    boneToModify {boneName=mstr_jacket_front_01_r}
    rotation=<rot 5 3.855 10.234>
   }
   n40 = "Transform (Modify) Bone - Bone: mstr_jacket_front_01_r"(ComponentPose=@n39, Alpha=Get pawnSpeedRange(), Translation="0,0,0", Rotation="0,0,0", Scale="1,1,1") {
    rotationMode=BMM_Additive
    rotationSpace=BCS_BoneSpace
    translationMode=BMM_Additive
    translationSpace=BCS_BoneSpace
    alphaBoolBlend {blendOption=Linear}
    alphaScaleBiasClamp {bInterpResult=true}
    boneToModify {boneName=mstr_jacket_front_01_r}
   }
   n41 = "Transform (Modify) Bone - Bone: mstr_jacket_side_01_l"(ComponentPose=@n40, Alpha=Get movingLeftSideAlpha(), Translation="0,0,0", Rotation="7.403,0.032,0.051", Scale="1,1,1") {
    rotationMode=BMM_Additive
    rotationSpace=BCS_BoneSpace
    translationMode=BMM_Additive
    translationSpace=BCS_BoneSpace
    alphaBoolBlend {blendOption=Linear}
    alphaScaleBiasClamp {bInterpResult=true}
    boneToModify {boneName=mstr_jacket_side_01_l}
    rotation=<rot 7.403 0.032 0.051>
   }
   n42 = "Transform (Modify) Bone - Bone: mstr_jacket_front_01_l"(ComponentPose=@n41, Alpha=Get IsCrouchingAlpha(), Translation="-5,-5,0", Rotation="9.8,10,1.7", Scale="1,1,1") {
    rotationMode=BMM_Additive
    rotationSpace=BCS_BoneSpace
    translationMode=BMM_Additive
    translationSpace=BCS_BoneSpace
    alphaBoolBlend {blendOption=Linear}
    boneToModify {boneName=mstr_jacket_front_01_l}
    rotation=<rot 9.8 10 1.7>
    translation=<-5 -5>
   }
   n43 = "Transform (Modify) Bone - Bone: mstr_jacket_front_01_l"(ComponentPose=@n42, Alpha=Get Riding Grindrail() -> (Riding Grindrail:double), Translation="0,0,0", Rotation="5,-52,-40", Scale="0,1,1") {
    rotationMode=BMM_Additive
    rotationSpace=BCS_BoneSpace
    alphaBoolBlend {blendOption=Linear}
    boneToModify {boneName=mstr_jacket_front_01_l}
    rotation=<rot 5 -52 -40>
    scale {x=0}
   }
   Save cached pose 'RBANPose'(Pose=Blend Poses by bool(
     BlendPose_0=Component To Local(ComponentPose=@n30)
     BlendPose_1=Component To Local(ComponentPose=@n43)
     BlendTime_0=0.01
     BlendTime_1=0.3
     bActiveValue=Get bIsSliding()
    ) {blendType=Linear}) {cacheName=RBANPose}
   n44 = Control Rig!(
    Source=Copy Pose From Mesh() {bUseAttachedParent=true}
    Alpha
    Crouching Walking:double=Get CrouchingMovingAlpha() -> (CrouchingMovingAlpha:double)
    Crouching:double=Get IsCrouchingAlpha()
    Thigh Rot L:double=Get thigh_rot_up_l()
    Thigh Rot R:double=Get thigh_rot_up_r()
    Skydiving:bool=Get bIsSkyDiving()
    CrouchingBool:bool=Get bIsCrouching()
    Jumping:double=Get JumpingAlpha()
   ) -> (Pose) {
    alphaBoolBlend {blendOption=Linear}
    controlRigAssetReference {blueprintRigClass=&/M_MED_RhombCamo_CtrlRig.M_MED_RhombCamo_CtrlRig_C}
   }
   Save cached pose 'PrePose'(Pose=@n44) {cacheName=PrePose}
   Get movingRightSideAlpha()
   Get Gravity Override()
   Get CrouchingMovingAlpha()
   Get IsCrouchingAlpha()
   Get pawnSpeedRange()
   Get pawnSpeedRange()
  ]}
 EventGraph {nodes [
   Event BlueprintInitializeAnimation()
   > Set Skel Mesh(Skel Mesh=GetOwnerSkeletalMeshComponent())
   n1 = TryGetPawnOwner()
   > Is Valid(InputObject=@n1)
   > n2 = Cast To FortPlayerPawnAthena(Object=@n1)
   > Set PlayerPawn(PlayerPawn:Fort Player Pawn Athena Object Reference=@n2) -> (Output_Get:Fort Player Pawn Athena Object Reference)
   Event OnCharacterCustomizationLoaded()
   > n3 = Set Skel Mesh(Skel Mesh=GetPartSkeletalMeshComponent())
   > n4 = Get Dynamic Material Instance(SkeletalMeshComponent=@n3)
   > Set Mat Instance Dynamic(Mat Instance Dynamic=@n4)
   > n5 = GetOrCreateDynamicOverlayMaterial(MeshComponent=@n3, MaterialName=None)
   > Set Overlay Mat Instance Dynamic(Overlay Mat Instance Dynamic=@n5)
   n6 = Event BlueprintUpdateAnimation()
   > Set Delta Time X(Delta Time X:double=@n6.DeltaTimeX) -> (Output_Get:double)
   > n7 = Sequence!(execute) -> (then_0, then_1, then_2:Exec, then_3:Exec)
   > Set pawnSpeedRange(pawnSpeedRange:double=MapRangeClamped(Value=Get PawnSpeed2D(), InRangeA=300, InRangeB=550)) -> (Output_Get:double)
   Set thigh_rot_up_l(
    execute=@n7.then_1
    thigh_rot_up_l:double=GetAlphaFromReferenceTransformDelta(
     SkeletalMeshComponent=Get Skel Mesh()
     BoneName=thigh_l
     BaseName=pelvis
     DeltaComponentType=QuaternionTwist
     InputRange=(X=0,Y=100)
     TwistAxis="0,0,1"
    )
   ) -> (Output_Get:double)
   > Set thigh_rot_up_01_l(thigh_rot_up_01_l:double=GetAlphaFromReferenceTransformDelta(
     SkeletalMeshComponent=Get Skel Mesh()
     BoneName=thigh_l
     BaseName=pelvis
     DeltaComponentType=QuaternionTwist
     InputRange=(X=0,Y=50)
     TwistAxis="0,0,1"
    )) -> (Output_Get:double)
   > Set thigh_rot_up_02_l(thigh_rot_up_02_l:double=GetAlphaFromReferenceTransformDelta(
     SkeletalMeshComponent=Get Skel Mesh()
     BoneName=thigh_l
     BaseName=pelvis
     DeltaComponentType=QuaternionTwist
     InputRange=(X=40,Y=100)
     TwistAxis="0,0,1"
    )) -> (Output_Get:double)
   > Set thigh_rot_up_r(thigh_rot_up_r:double=GetAlphaFromReferenceTransformDelta(
     SkeletalMeshComponent=Get Skel Mesh()
     BoneName=thigh_r
     BaseName=pelvis
     DeltaComponentType=QuaternionTwist
     InputRange=(X=0,Y=100)
     TwistAxis="0,0,1"
    )) -> (Output_Get:double)
   Set pushOutLeftJacketAlpha(execute=@n7.then_2, pushOutLeftJacketAlpha:double=float * float() {is=title K2Node_PromotableOperator}) -> (Output_Get:double)
   > Branch(Condition=AND Boolean(A=Get bIsAccelerating2D(), B=Get bIsOnGround() -> (bIsOnGround:bool)))
   n8 = Sequence!(execute=@n7.then_3) -> (then_0, then_1, then_2, then_3, then_4:Exec, then_5:Exec)
   n9 = Gravity Override State Function DEPRECATED(
    execute=@n8.then_1
    CharacterPartAnimInstance=Self-Reference()
    GameTimeInSeconds=GetGameTimeInSeconds()
    GravityOverrideParams=Get Gravity Override Params()
   )
   > Set Gravity Override(Gravity Override=@n9)
   > Set Component Linear Acc Scale(Component Linear Acc Scale:Vector=SelectVector(A="0.2,0.2,0.2", B="0.5,0.5,0.5", bPickA=Get bIsSkyDiving())) -> (Output_Get:Vector)
   > Set Component Linear Vel Scale(Component Linear Vel Scale:Vector=SelectVector(A="0.1,0.1,0.1", B="0.3,0.3,0.3", bPickA=Get bIsSkyDiving())) -> (Output_Get:Vector)
   n10 = AND Boolean(A=Get bIsCrouching(), B=NOT Boolean(A=Get bIsMoving2D()))
   Set IsCrouchingAlpha(execute=@n8.then_2, IsCrouchingAlpha:double="To Float (Boolean)"(InBool=@n10)) -> (Output_Get:double)
   n11 = Branch(execute=@n8.then_4, Condition=AND Boolean(A=Get bIsCrouching(), B=Get bIsMoving2D()))
   > n12 = Set CrouchingMovingAlpha(CrouchingMovingAlpha:double=1) -> (Output_Get:double)
   Set CrouchingMovingAlpha(
    execute=@n11.else
    CrouchingMovingAlpha=FInterpTo(Current=@n12, Target=0, DeltaTime=Get Delta Time X() -> (Delta Time X:double), InterpSpeed=20)
   )
   Is Valid(exec=@n8.then_0, InputObject=Get PlayerPawn() -> (PlayerPawn:Fort Player Pawn Athena Object Reference))
   > Set WPO Globals Function(
    Overlay Material=Get Overlay Mat Instance Dynamic()
    Skeletal Mesh=Get Skel Mesh()
    "Local Noise Transform  Location"="Get Local Noise Transform  Location"() -> ("Local Noise Transform  Location":Vector)
    Local Noise Offset=Get Local Noise Offset()
   )
   n13 = Get Delta Time X()
   > Set WPO Regional Function!(
    execute
    Material:Material Instance Dynamic Object Reference=Get Mat Instance Dynamic()
    Overlay Material:Material Instance Dynamic Object Reference=Get Overlay Mat Instance Dynamic()
    Skeletal Mesh:Skeletal Mesh Component Object Reference=Get Skel Mesh()
    Velocity Transform:Name=foot_r
    Velocity Interp Speed:double=Get Velocity Interp Speed()
    Collision Transform:Name=calf_r
    Collision Transform Axis:EAxis Enum
    Invert Collision Transform Axis:bool=false
    Delta Time X:double=@n13
    Previous Velocity Transform Location:"Vector (by ref)"=Get R1 Velocity Transform Location()
    Previous Velocity Transform Velocity:"Vector (by ref)"=Get R1 Velocity Transform Velocity()
    Region:int=1
   ) -> (then)
   > Set WPO Regional Function(
    Material=Get Mat Instance Dynamic()
    Overlay Material=Get Overlay Mat Instance Dynamic()
    Skeletal Mesh=Get Skel Mesh()
    Velocity Transform=foot_l
    Velocity Interp Speed=Get Velocity Interp Speed()
    Collision Transform=calf_l
    Invert Collision Transform Axis=true
    Delta Time X=@n13
    Previous Velocity Transform Location=Get R2 Velocity Transform Location()
    Previous Velocity Transform Velocity=Get R2 Velocity Transform Velocity()
    Region=2
   )
   > n14 = Branch(Condition=Get bIsSkydiveFloating())
   > n15 = Set WPO Pho Wind Intensity(WPO Pho Wind Intensity=1)
   SetScalarParameterValue(execute=[@n15.then, @n16.then], self=Get Mat Instance Dynamic(), ParameterName=ClothSim_WindStrength, Value=Get WPO Pho Wind Intensity())
   n16 = Set WPO Pho Wind Intensity(execute=@n14.else, WPO Pho Wind Intensity=0)
   n17 = Get bIsCrouching()
   Set StandingAndNotCrouching(execute=@n8.then_5, StandingAndNotCrouching:bool=AND Boolean(A=Get IsStanding() -> (IsStanding:bool), B=NOT Boolean(A=@n17))) -> (Output_Get:bool)
   > Set SkyDivingBackwardsMovementAlpha(SkyDivingBackwardsMovementAlpha:double=float * float(A:bool=Get bIsSkydiveFloating(), B=Get BackwardMovementAlpha() -> (BackwardMovementAlpha:double))) -> (Output_Get:double)
   > Set Riding Grindrail(Riding Grindrail:double="To Float (Boolean)"(InBool=Get bIsOnGrindRail())) -> (Output_Get:double)
   Set CrouchingStandingLookAtAlpha(execute=@n8.then_3, CrouchingStandingLookAtAlpha:double=SelectFloat(A=0.85, B=0, bPickA=@n10)) -> (Output_Get:double)
   > n18 = Branch(Condition=@n10)
   > n19 = Set CrouchingStanding(CrouchingStanding:double=1) -> (Output_Get:double)
   Set CrouchingStanding(execute=@n18.else, CrouchingStanding=FInterpTo(Current=@n19, Target=0, DeltaTime=Get Delta Time X(), InterpSpeed=20))
   Get Overlay Mat Instance Dynamic()
   Get bIsSkyDiving()
  ]}
]
variables [
 BackwardMovementAlpha {type=number}
 BwdMovement {type=number}
 Component Linear Acc Scale {type=Vector}
 Component Linear Vel Scale {type=Vector}
 CrouchingMovingAlpha {type=number}
 CrouchingStanding {type=number}
 CrouchingStandingLookAtAlpha {type=number}
 Delta Time X {type=number}
 Gravity Override {type=Vector}
 Gravity Override Params {
  type=GravityOverrideParamsStruct
  default {
   skydive_FloatingDown {gravityOverride=<0 -700 -100>, windAmplitude=<200 -500 200>, windOffset=<0 -600 100>}
   skydive_GoingUp {gravityOverride=<0 0 0>, windAmplitude=<100 -250 150>}
   skydive_Parachuting {gravityOverride=<0 -500 -980>, windAmplitude=<200 -150 400>, windFrequency=<1.5 4 6>}
  }
 }
 IsCrouchingAlpha {type=number}
 IsStanding {type=boolean}
 JumpingAlpha {type=number}
 Local Noise Offset {type=Vector}
 "Local Noise Transform  Location" {type=Vector}
 Mat Instance Dynamic {type=MaterialInstanceDynamic}
 movingLeftSideAlpha {type=number}
 movingRightSideAlpha {type=number}
 Overlay Mat Instance Dynamic {type=MaterialInstanceDynamic}
 pawnSpeedRange {type=number}
 PlayerPawn {type=FortPlayerPawnAthena}
 pushOutLeftJacketAlpha {type=number}
 R1 Velocity Transform Location {type=Vector}
 R1 Velocity Transform Velocity {type=Vector}
 R2 Velocity Transform Location {type=Vector}
 R2 Velocity Transform Velocity {type=Vector}
 Riding Grindrail {type=number}
 Skel Mesh {type=SkeletalMeshComponent}
 SkyDivingBackwardsMovementAlpha {type=number}
 StandingAndNotCrouching {type=boolean}
 thigh_rot_up_01_l {type=number}
 thigh_rot_up_02_l {type=number}
 thigh_rot_up_l {type=number}
 thigh_rot_up_r {type=number}
 Velocity Interp Speed {type=number, default=1}
 WPO Pho Wind Intensity {type=number}
]
