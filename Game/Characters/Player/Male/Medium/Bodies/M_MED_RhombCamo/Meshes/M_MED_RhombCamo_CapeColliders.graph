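# Physics asset for the RhombCamo medium male player body, named for cape collision, with colliders on the upper body and jacket bones.
# Capsules cover pelvis, spine, neck and clavicles, with a sphere for the head; arm and leg bodies have no shapes.
# Master jacket bones branch from the pelvis into seven four-segment capsule chains: front, side and rear per side, plus center.
# Constraints follow this hierarchy with all swing and twist motion free.
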PhysicsAsset /Game/Characters/Player/Male/Medium/Bodies/M_MED_RhombCamo/Meshes/M_MED_RhombCamo_CapeColliders
bodies {
 calf_l []
 calf_r []
 clavicle_l [{shapeType=Capsule, length=9.0192, radius=10.5, center=<9.4266 -1.1806 -2.4977>, rotation=<rot 69.3297 1.4476 176.9462>}]
 clavicle_r [{shapeType=Capsule, length=9.0192, radius=10.5, center=<-9.4266 1.1806 2.4977>, rotation=<rot 69.3297 1.4476 -3.0538>}]
 dyn_jacket_center_01 [{shapeType=Capsule, length=15, radius=9, shapeName=dyn_jacket_center_01_capsule, rotation=<rot -0.0003 0 80>}]
 dyn_jacket_center_02 [{shapeType=Capsule, length=15, radius=9, shapeName=dyn_jacket_center_02_capsule, rotation=<rot -0.0003 0 80>}]
 dyn_jacket_center_03 [{shapeType=Capsule, length=15, radius=9, shapeName=dyn_jacket_center_03_capsule, rotation=<rot -0.0001 0 -100>}]
 dyn_jacket_center_04 [{shapeType=Capsule, length=15, radius=9, shapeName=dyn_jacket_center_04_capsule, rotation=<rot -0.0001 0 -100>}]
 dyn_jacket_front_01_l [{shapeType=Capsule, length=15, radius=9, shapeName=dyn_jacket_front_01_l_capsule, rotation=<rot 0.0001 0 90>}]
 dyn_jacket_front_01_r [{shapeType=Capsule, length=15, radius=9, shapeName=dyn_jacket_front_01_r_capsule, rotation=<rot 0.0001 0 90>}]
 dyn_jacket_front_02_l [{shapeType=Capsule, length=15, radius=9, shapeName=dyn_jacket_front_02_l_capsule, rotation=<rot 0.0001 0 90>}]
 dyn_jacket_front_02_r [{shapeType=Capsule, length=15, radius=9, shapeName=dyn_jacket_front_02_r_capsule, rotation=<rot 0.0001 0 90>}]
 dyn_jacket_front_03_l [{shapeType=Capsule, length=15, radius=9, shapeName=dyn_jacket_front_03_l_capsule, rotation=<rot 0.0001 0 90>}]
 dyn_jacket_front_03_r [{shapeType=Capsule, length=15, radius=9, shapeName=dyn_jacket_front_03_r_capsule, rotation=<rot 0.0001 0 90>}]
 dyn_jacket_front_04_l [{shapeType=Capsule, length=15, radius=9, shapeName=dyn_jacket_front_04_l_capsule, rotation=<rot 0.0001 0 90>}]
 dyn_jacket_front_04_r [{shapeType=Capsule, length=15, radius=9, shapeName=dyn_jacket_front_04_r_capsule, rotation=<rot 0.0001 0 90>}]
 dyn_jacket_rear_01_l [{shapeType=Capsule, length=15, radius=9, shapeName=dyn_jacket_rear_01_l_capsule, rotation=<rot -0.0001 0 90>}]
 dyn_jacket_rear_01_r [{shapeType=Capsule, length=15, radius=9, shapeName=dyn_jacket_rear_01_r_capsule, rotation=<rot -0.0001 0 90>}]
 dyn_jacket_rear_02_l [{shapeType=Capsule, length=15, radius=9, shapeName=dyn_jacket_rear_02_l_capsule, rotation=<rot -0.0001 0 90>}]
 dyn_jacket_rear_02_r [{shapeType=Capsule, length=15, radius=9, shapeName=dyn_jacket_rear_02_r_capsule, rotation=<rot -0.0001 0 90>}]
 dyn_jacket_rear_03_l [{shapeType=Capsule, length=15, radius=9, shapeName=dyn_jacket_rear_03_l_capsule, rotation=<rot -0.0001 0 90>}]
 dyn_jacket_rear_03_r [{shapeType=Capsule, length=15, radius=9, shapeName=dyn_jacket_rear_03_r_capsule, rotation=<rot -0.0001 0 90>}]
 dyn_jacket_rear_04_l [{shapeType=Capsule, length=15, radius=9, shapeName=dyn_jacket_rear_04_l_capsule, rotation=<rot -0.0001 0 90>}]
 dyn_jacket_rear_04_r [{shapeType=Capsule, length=15, radius=9, shapeName=dyn_jacket_rear_04_r_capsule, rotation=<rot -0.0001 0 90>}]
 dyn_jacket_side_01_l [{shapeType=Capsule, length=15, radius=9, shapeName=dyn_jacket_side_01_l_capsule, rotation=<rot 0 0.0001 90>}]
 dyn_jacket_side_01_r [{shapeType=Capsule, length=15, radius=9, shapeName=dyn_jacket_side_01_r_capsule, rotation=<rot 0 -0.0001 90>}]
 dyn_jacket_side_02_l [{shapeType=Capsule, length=15, radius=9, shapeName=dyn_jacket_side_02_l_capsule, rotation=<rot 0 0.0001 90>}]
 dyn_jacket_side_02_r [{shapeType=Capsule, length=15, radius=9, shapeName=dyn_jacket_side_02_r_capsule, rotation=<rot 0 -0.0001 90>}]
 dyn_jacket_side_03_l [{shapeType=Capsule, length=15, radius=9, shapeName=dyn_jacket_side_03_l_capsule, rotation=<rot 0 0.0001 90>}]
 dyn_jacket_side_03_r [{shapeType=Capsule, length=15, radius=9, shapeName=dyn_jacket_side_03_r_capsule, rotation=<rot 0 -0.0001 90>}]
 dyn_jacket_side_04_l [{shapeType=Capsule, length=15, radius=9, shapeName=dyn_jacket_side_04_l_capsule, rotation=<rot 0 0.0001 90>}]
 dyn_jacket_side_04_r [{shapeType=Capsule, length=15, radius=9, shapeName=dyn_jacket_side_04_r_capsule, rotation=<rot 0 -0.0001 90>}]
 foot_l []
 foot_r []
 head [{shapeType=Sphere, radius=17.3125, center=<11.028 -0.174 0>}]
 lowerarm_l []
 lowerarm_r []
 mstr_Jacket [{shapeType=Capsule, length=1, radius=1, shapeName=mstr_Jacket_capsule}]
 mstr_Jacket_Left [{shapeType=Capsule, length=1, radius=1, shapeName=mstr_Jacket_Left_capsule}]
 mstr_Jacket_Right [{shapeType=Capsule, length=1, radius=1, shapeName=mstr_Jacket_Right_capsule, rotation=<rot 0 -0.0001 90>}]
 mstr_jacket_center_01_l1 [{shapeType=Capsule, length=1, radius=1, shapeName=mstr_jacket_center_01_l1_capsule, rotation=<rot -0.0001 0 90>}]
 mstr_jacket_front_01_l [{shapeType=Capsule, length=15, radius=9, shapeName=mstr_jacket_front_01_l_capsule}]
 mstr_jacket_front_01_r [{shapeType=Capsule, length=15, radius=9, shapeName=mstr_jacket_front_01_r_capsule, rotation=<rot 0.0001 0 90>}]
 mstr_jacket_rear_01_l [{shapeType=Capsule, length=1, radius=1, shapeName=mstr_jacket_rear_01_l_capsule, rotation=<rot -0.0001 0 90>}]
 mstr_jacket_rear_01_r [{shapeType=Capsule, length=1, radius=1, shapeName=mstr_jacket_rear_01_r_capsule, rotation=<rot -0.0001 0 90>}]
 mstr_jacket_side_01_l [{shapeType=Capsule, length=1, radius=1, shapeName=mstr_jacket_side_01_l_capsule, rotation=<rot 0 0.0001 90>}]
 mstr_jacket_side_01_r [{shapeType=Capsule, length=1, radius=1, shapeName=mstr_jacket_side_01_r_capsule, rotation=<rot 0 -0.0001 90>}]
 neck_02 [{shapeType=Capsule, length=13.4375, radius=7.3175, center=<-0.2642 -0.3408 0>, rotation=<rot 90 -171.1193 -175.3978>}]
 pelvis [{shapeType=Capsule, length=7, radius=10.4, shapeName=spine_01_capsule, center=<-0.4269 -0.4086 -0.1905>}]
 spine_01 [{shapeType=Capsule, length=5.4438, radius=13.0586, center=<-1.7965 0.8851 0>}]
 spine_02 [{shapeType=Capsule, length=4.7826, radius=13.6322, center=<-1.711 2.4708 0.0001>}]
 spine_03 [{shapeType=Capsule, length=7.7107, radius=12.2031, center=<0.1445 1.3501 0.0001>}]
 spine_04 [{shapeType=Capsule, length=8.4404, radius=14.6325, center=<2.766 1.9588 0.0002>}]
 spine_05 [{shapeType=Capsule, length=5.5226, radius=16.2931, center=<-0.6513 3.0377 0>}]
 thigh_l []
 thigh_r []
 upperarm_l []
 upperarm_r []
}
constraints [
 {
  bone1Name=spine_01
  bone2Name=pelvis
  swing1LimitDegrees=45
  swing1Motion=Free
  swing2LimitDegrees=45
  swing2Motion=Free
  twistLimitDegrees=45
  twistMotion=Free
 }
 {
  bone1Name=spine_02
  bone2Name=spine_01
  swing1LimitDegrees=45
  swing1Motion=Free
  swing2LimitDegrees=45
  swing2Motion=Free
  twistLimitDegrees=45
  twistMotion=Free
 }
 {
  bone1Name=spine_03
  bone2Name=spine_02
  swing1LimitDegrees=45
  swing1Motion=Free
  swing2LimitDegrees=45
  swing2Motion=Free
  twistLimitDegrees=45
  twistMotion=Free
 }
 {
  bone1Name=spine_04
  bone2Name=spine_03
  swing1LimitDegrees=45
  swing1Motion=Free
  swing2LimitDegrees=45
  swing2Motion=Free
  twistLimitDegrees=45
  twistMotion=Free
 }
 {
  bone1Name=spine_05
  bone2Name=spine_04
  swing1LimitDegrees=45
  swing1Motion=Free
  swing2LimitDegrees=45
  swing2Motion=Free
  twistLimitDegrees=45
  twistMotion=Free
 }
 {
  bone1Name=neck_02
  bone2Name=spine_05
  swing1LimitDegrees=45
  swing1Motion=Free
  swing2LimitDegrees=45
  swing2Motion=Free
  twistLimitDegrees=45
  twistMotion=Free
 }
 {
  bone1Name=head
  bone2Name=neck_02
  swing1LimitDegrees=45
  swing1Motion=Free
  swing2LimitDegrees=45
  swing2Motion=Free
  twistLimitDegrees=45
  twistMotion=Free
 }
 {
  bone1Name=thigh_l
  bone2Name=pelvis
  swing1LimitDegrees=45
  swing1Motion=Free
  swing2LimitDegrees=45
  swing2Motion=Free
  twistLimitDegrees=45
  twistMotion=Free
 }
 {
  bone1Name=calf_l
  bone2Name=thigh_l
  swing1LimitDegrees=45
  swing1Motion=Free
  swing2LimitDegrees=45
  swing2Motion=Free
  twistLimitDegrees=45
  twistMotion=Free
 }
 {
  bone1Name=foot_l
  bone2Name=calf_l
  swing1LimitDegrees=45
  swing1Motion=Free
  swing2LimitDegrees=45
  swing2Motion=Free
  twistLimitDegrees=45
  twistMotion=Free
 }
 {
  bone1Name=clavicle_r
  bone2Name=spine_05
  swing1LimitDegrees=45
  swing1Motion=Free
  swing2LimitDegrees=45
  swing2Motion=Free
  twistLimitDegrees=45
  twistMotion=Free
 }
 {
  bone1Name=upperarm_r
  bone2Name=clavicle_r
  swing1LimitDegrees=45
  swing1Motion=Free
  swing2LimitDegrees=45
  swing2Motion=Free
  twistLimitDegrees=45
  twistMotion=Free
 }
 {
  bone1Name=lowerarm_r
  bone2Name=upperarm_r
  swing1LimitDegrees=45
  swing1Motion=Free
  swing2LimitDegrees=45
  swing2Motion=Free
  twistLimitDegrees=45
  twistMotion=Free
 }
 {
  bone1Name=foot_r
  bone2Name=calf_r
  swing1LimitDegrees=45
  swing1Motion=Free
  swing2LimitDegrees=45
  swing2Motion=Free
  twistLimitDegrees=45
  twistMotion=Free
 }
 {
  bone1Name=lowerarm_l
  bone2Name=upperarm_l
  swing1LimitDegrees=45
  swing1Motion=Free
  swing2LimitDegrees=45
  swing2Motion=Free
  twistLimitDegrees=45
  twistMotion=Free
 }
 {
  bone1Name=upperarm_l
  bone2Name=clavicle_l
  swing1LimitDegrees=45
  swing1Motion=Free
  swing2LimitDegrees=45
  swing2Motion=Free
  twistLimitDegrees=45
  twistMotion=Free
 }
 {
  bone1Name=clavicle_l
  bone2Name=spine_05
  swing1LimitDegrees=45
  swing1Motion=Free
  swing2LimitDegrees=45
  swing2Motion=Free
  twistLimitDegrees=45
  twistMotion=Free
 }
 {
  bone1Name=calf_r
  bone2Name=thigh_r
  swing1LimitDegrees=45
  swing1Motion=Free
  swing2LimitDegrees=45
  swing2Motion=Free
  twistLimitDegrees=45
  twistMotion=Free
 }
 {
  bone1Name=thigh_r
  bone2Name=pelvis
  swing1LimitDegrees=45
  swing1Motion=Free
  swing2LimitDegrees=45
  swing2Motion=Free
  twistLimitDegrees=45
  twistMotion=Free
 }
 {
  bone1Name=mstr_Jacket
  bone2Name=pelvis
  swing1LimitDegrees=45
  swing1Motion=Free
  swing2LimitDegrees=45
  swing2Motion=Free
  twistLimitDegrees=45
  twistMotion=Free
 }
 {
  bone1Name=mstr_Jacket_Left
  bone2Name=mstr_Jacket
  swing1LimitDegrees=45
  swing1Motion=Free
  swing2LimitDegrees=45
  swing2Motion=Free
  twistLimitDegrees=45
  twistMotion=Free
 }
 {
  bone1Name=mstr_jacket_front_01_l
  bone2Name=mstr_Jacket_Left
  swing1LimitDegrees=45
  swing1Motion=Free
  swing2LimitDegrees=45
  swing2Motion=Free
  twistLimitDegrees=45
  twistMotion=Free
 }
 {
  bone1Name=dyn_jacket_front_01_l
  bone2Name=mstr_jacket_front_01_l
  swing1LimitDegrees=45
  swing1Motion=Free
  swing2LimitDegrees=45
  swing2Motion=Free
  twistLimitDegrees=45
  twistMotion=Free
 }
 {
  bone1Name=dyn_jacket_front_02_l
  bone2Name=dyn_jacket_front_01_l
  swing1LimitDegrees=45
  swing1Motion=Free
  swing2LimitDegrees=45
  swing2Motion=Free
  twistLimitDegrees=45
  twistMotion=Free
 }
 {
  bone1Name=dyn_jacket_front_03_l
  bone2Name=dyn_jacket_front_02_l
  swing1LimitDegrees=45
  swing1Motion=Free
  swing2LimitDegrees=45
  swing2Motion=Free
  twistLimitDegrees=45
  twistMotion=Free
 }
 {
  bone1Name=dyn_jacket_front_04_l
  bone2Name=dyn_jacket_front_03_l
  swing1LimitDegrees=45
  swing1Motion=Free
  swing2LimitDegrees=45
  swing2Motion=Free
  twistLimitDegrees=45
  twistMotion=Free
 }
 {
  bone1Name=mstr_jacket_rear_01_l
  bone2Name=mstr_Jacket_Left
  swing1LimitDegrees=45
  swing1Motion=Free
  swing2LimitDegrees=45
  swing2Motion=Free
  twistLimitDegrees=45
  twistMotion=Free
 }
 {
  bone1Name=dyn_jacket_rear_01_l
  bone2Name=mstr_jacket_rear_01_l
  swing1LimitDegrees=45
  swing1Motion=Free
  swing2LimitDegrees=45
  swing2Motion=Free
  twistLimitDegrees=45
  twistMotion=Free
 }
 {
  bone1Name=dyn_jacket_rear_02_l
  bone2Name=dyn_jacket_rear_01_l
  swing1LimitDegrees=45
  swing1Motion=Free
  swing2LimitDegrees=45
  swing2Motion=Free
  twistLimitDegrees=45
  twistMotion=Free
 }
 {
  bone1Name=dyn_jacket_rear_03_l
  bone2Name=dyn_jacket_rear_02_l
  swing1LimitDegrees=45
  swing1Motion=Free
  swing2LimitDegrees=45
  swing2Motion=Free
  twistLimitDegrees=45
  twistMotion=Free
 }
 {
  bone1Name=dyn_jacket_rear_04_l
  bone2Name=dyn_jacket_rear_03_l
  swing1LimitDegrees=45
  swing1Motion=Free
  swing2LimitDegrees=45
  swing2Motion=Free
  twistLimitDegrees=45
  twistMotion=Free
 }
 {
  bone1Name=mstr_jacket_side_01_l
  bone2Name=mstr_Jacket_Left
  swing1LimitDegrees=45
  swing1Motion=Free
  swing2LimitDegrees=45
  swing2Motion=Free
  twistLimitDegrees=45
  twistMotion=Free
 }
 {
  bone1Name=dyn_jacket_side_01_l
  bone2Name=mstr_jacket_side_01_l
  swing1LimitDegrees=45
  swing1Motion=Free
  swing2LimitDegrees=45
  swing2Motion=Free
  twistLimitDegrees=45
  twistMotion=Free
 }
 {
  bone1Name=dyn_jacket_side_02_l
  bone2Name=dyn_jacket_side_01_l
  swing1LimitDegrees=45
  swing1Motion=Free
  swing2LimitDegrees=45
  swing2Motion=Free
  twistLimitDegrees=45
  twistMotion=Free
 }
 {
  bone1Name=dyn_jacket_side_03_l
  bone2Name=dyn_jacket_side_02_l
  swing1LimitDegrees=45
  swing1Motion=Free
  swing2LimitDegrees=45
  swing2Motion=Free
  twistLimitDegrees=45
  twistMotion=Free
 }
 {
  bone1Name=dyn_jacket_side_04_l
  bone2Name=dyn_jacket_side_03_l
  swing1LimitDegrees=45
  swing1Motion=Free
  swing2LimitDegrees=45
  swing2Motion=Free
  twistLimitDegrees=45
  twistMotion=Free
 }
 {
  bone1Name=mstr_Jacket_Right
  bone2Name=mstr_Jacket
  swing1LimitDegrees=45
  swing1Motion=Free
  swing2LimitDegrees=45
  swing2Motion=Free
  twistLimitDegrees=45
  twistMotion=Free
 }
 {
  bone1Name=mstr_jacket_front_01_r
  bone2Name=mstr_Jacket_Right
  swing1LimitDegrees=45
  swing1Motion=Free
  swing2LimitDegrees=45
  swing2Motion=Free
  twistLimitDegrees=45
  twistMotion=Free
 }
 {
  bone1Name=dyn_jacket_front_01_r
  bone2Name=mstr_jacket_front_01_r
  swing1LimitDegrees=45
  swing1Motion=Free
  swing2LimitDegrees=45
  swing2Motion=Free
  twistLimitDegrees=45
  twistMotion=Free
 }
 {
  bone1Name=dyn_jacket_front_02_r
  bone2Name=dyn_jacket_front_01_r
  swing1LimitDegrees=45
  swing1Motion=Free
  swing2LimitDegrees=45
  swing2Motion=Free
  twistLimitDegrees=45
  twistMotion=Free
 }
 {
  bone1Name=dyn_jacket_front_03_r
  bone2Name=dyn_jacket_front_02_r
  swing1LimitDegrees=45
  swing1Motion=Free
  swing2LimitDegrees=45
  swing2Motion=Free
  twistLimitDegrees=45
  twistMotion=Free
 }
 {
  bone1Name=dyn_jacket_front_04_r
  bone2Name=dyn_jacket_front_03_r
  swing1LimitDegrees=45
  swing1Motion=Free
  swing2LimitDegrees=45
  swing2Motion=Free
  twistLimitDegrees=45
  twistMotion=Free
 }
 {
  bone1Name=mstr_jacket_rear_01_r
  bone2Name=mstr_Jacket_Right
  swing1LimitDegrees=45
  swing1Motion=Free
  swing2LimitDegrees=45
  swing2Motion=Free
  twistLimitDegrees=45
  twistMotion=Free
 }
 {
  bone1Name=dyn_jacket_rear_01_r
  bone2Name=mstr_jacket_rear_01_r
  swing1LimitDegrees=45
  swing1Motion=Free
  swing2LimitDegrees=45
  swing2Motion=Free
  twistLimitDegrees=45
  twistMotion=Free
 }
 {
  bone1Name=dyn_jacket_rear_02_r
  bone2Name=dyn_jacket_rear_01_r
  swing1LimitDegrees=45
  swing1Motion=Free
  swing2LimitDegrees=45
  swing2Motion=Free
  twistLimitDegrees=45
  twistMotion=Free
 }
 {
  bone1Name=dyn_jacket_rear_03_r
  bone2Name=dyn_jacket_rear_02_r
  swing1LimitDegrees=45
  swing1Motion=Free
  swing2LimitDegrees=45
  swing2Motion=Free
  twistLimitDegrees=45
  twistMotion=Free
 }
 {
  bone1Name=dyn_jacket_rear_04_r
  bone2Name=dyn_jacket_rear_03_r
  swing1LimitDegrees=45
  swing1Motion=Free
  swing2LimitDegrees=45
  swing2Motion=Free
  twistLimitDegrees=45
  twistMotion=Free
 }
 {
  bone1Name=mstr_jacket_side_01_r
  bone2Name=mstr_Jacket_Right
  swing1LimitDegrees=45
  swing1Motion=Free
  swing2LimitDegrees=45
  swing2Motion=Free
  twistLimitDegrees=45
  twistMotion=Free
 }
 {
  bone1Name=dyn_jacket_side_01_r
  bone2Name=mstr_jacket_side_01_r
  swing1LimitDegrees=45
  swing1Motion=Free
  swing2LimitDegrees=45
  swing2Motion=Free
  twistLimitDegrees=45
  twistMotion=Free
 }
 {
  bone1Name=dyn_jacket_side_02_r
  bone2Name=dyn_jacket_side_01_r
  swing1LimitDegrees=45
  swing1Motion=Free
  swing2LimitDegrees=45
  swing2Motion=Free
  twistLimitDegrees=45
  twistMotion=Free
 }
 {
  bone1Name=dyn_jacket_side_03_r
  bone2Name=dyn_jacket_side_02_r
  swing1LimitDegrees=45
  swing1Motion=Free
  swing2LimitDegrees=45
  swing2Motion=Free
  twistLimitDegrees=45
  twistMotion=Free
 }
 {
  bone1Name=dyn_jacket_side_04_r
  bone2Name=dyn_jacket_side_03_r
  swing1LimitDegrees=45
  swing1Motion=Free
  swing2LimitDegrees=45
  swing2Motion=Free
  twistLimitDegrees=45
  twistMotion=Free
 }
 {
  bone1Name=mstr_jacket_center_01_l1
  bone2Name=mstr_Jacket
  swing1LimitDegrees=45
  swing1Motion=Free
  swing2LimitDegrees=45
  swing2Motion=Free
  twistLimitDegrees=45
  twistMotion=Free
 }
 {
  bone1Name=dyn_jacket_center_01
  bone2Name=mstr_jacket_center_01_l1
  swing1LimitDegrees=45
  swing1Motion=Free
  swing2LimitDegrees=45
  swing2Motion=Free
  twistLimitDegrees=45
  twistMotion=Free
 }
 {
  bone1Name=dyn_jacket_center_02
  bone2Name=dyn_jacket_center_01
  swing1LimitDegrees=45
  swing1Motion=Free
  swing2LimitDegrees=45
  swing2Motion=Free
  twistLimitDegrees=45
  twistMotion=Free
 }
 {
  bone1Name=dyn_jacket_center_03
  bone2Name=dyn_jacket_center_02
  swing1LimitDegrees=45
  swing1Motion=Free
  swing2LimitDegrees=45
  swing2Motion=Free
  twistLimitDegrees=45
  twistMotion=Free
 }
 {
  bone1Name=dyn_jacket_center_04
  bone2Name=dyn_jacket_center_03
  swing1LimitDegrees=45
  swing1Motion=Free
  swing2LimitDegrees=45
  swing2Motion=Free
  twistLimitDegrees=45
  twistMotion=Free
 }
]
